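# Translucent UI material for the VanillaSoda weapon reticle, drawing a mirrored side shape and a center dot, both with shadows.
# An MSDF texture, mirrored through a symmetry function with size and distance parameters, is anti-aliased at two thresholds for shape and shadow.
# ProgressFill drives a stepped fill mask, while CooldownColor blends the fill from white to a vertical purple-and-blue gradient.
# Named reroutes carry the masks and color into the composites.
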Material /VanillaSoda/Gameplay/VanillaSodaWeapon/Art/Materials/M_UI_VanillaSoda_Reticle
graph {
 properties {
  bAutomaticallySetUsageInEditor=false
  blendMode=BLEND_Translucent
  materialDomain=MD_UI
  previewMesh=/Engine/Transient.MaterialEditorMeshComponent_37
  usedShadingModels=Default Lit
 }
 nodes [
  n1 = NamedRerouteUsage()
  n2 = NamedRerouteUsage()
  Add(A=LinearInterpolate(A=NamedRerouteUsage(), B=Multiply(A=@n1, B=@n2), Alpha=NamedRerouteUsage()), B=NamedRerouteUsage())
  n3 = NamedRerouteUsage()
  n4 = LinearInterpolate(A=@n3, B=Multiply(A=NamedRerouteUsage(), B=@n3)) {constAlpha=0.6}
  n5 = Multiply(A=NamedRerouteUsage(), B=ScalarParameter() {defaultValue=0.5, parameterName=ShapeShadowOpacity})
  n6 = NamedRerouteUsage()
  n7 = ScalarParameter() {defaultValue=0.3, parameterName=CircleShadowOpacity}
  Add(A=Add(A=Add(A=@n4, B=@n5), B=Multiply(A=@n6, B=@n7)), B=NamedRerouteUsage())
  n8 = ScalarParameter() {defaultValue=0.7, parameterName=ReticleSizeX}
  n9 = Constant() {r=0.4}
  n10 = MaterialFunctionCall!(
   UV=MaterialFunctionCall(UVs=TextureCoordinate(), Texture Scale=AppendVector(A=@n8, B=@n9)) {materialFunction=&ScaleUVsByCenter}.UVs
   Mirror Y
   Swizzle UV
  ) {materialFunction=/FortUILibrary/MaterialFunctions/Basics/MF_UI_Symmetry_2x1_New}
  n11 = Step(
   Y=ComponentMask(None=@n10.Flipped Symmetry) {r=true}
   X=LinearInterpolate(Alpha=ScalarParameter() {defaultValue=1, parameterName=ProgressFill}) {constA=0.37, constB=1.3}
  ) {constX=0}
  LinearInterpolate(A=@n11, B=NamedRerouteUsage())
  n12 = Constant4Vector() {constant=<rgba 0.2833 0.007 1 1>}
  n13 = Constant4Vector() {constant=<rgba 0 0.3178 1 1>}
  n14 = TextureCoordinate()
  n15 = Multiply(A=Add(A=ComponentMask(None=@n14) {g=true}) {constB=-0.1}) {constB=2}
  n16 = SmoothStep(Value=Abs(None=Add(A=@n15) {constB=-1})) {constMax=0.3}
  NamedRerouteDeclaration(None=LinearInterpolate(B=LinearInterpolate(A=@n12.RGB, B=@n13.RGB, Alpha=@n16), Alpha=ScalarParameter() {parameterName=CooldownColor}) {constA=1}) {name=FillColor, nodeColor=<rgba 0 0 0 1>}
  n17 = Constant() {r=0.035}
  n18 = MaterialFunctionCall(Size=Constant() {r=0.03}) {materialFunction=&MF_UI_Circle}
  NamedRerouteDeclaration(None=Subtract(
    A=MaterialFunctionCall(Size=@n17) {materialFunction=&MF_UI_Circle}.Fill
    B=NamedRerouteDeclaration(None=@n18.Fill) {name=MiddleCircle, nodeColor=<rgba 0 0 0 1>}
   )) {name=MiddleCircleShadow, nodeColor=<rgba 0 0 0 1>}
  NamedRerouteDeclaration(None=@n11) {name=ProgressFill, nodeColor=<rgba 0 0 0 1>}
  n19 = ScalarParameter() {defaultValue=0.35, parameterName=ReticleDistanceX}
  n20 = Constant() {r=-0.05}
  n21 = TextureSample(UVs=Add(A=@n10.Symmetry, B=AppendVector(A=@n19, B=@n20))) {
   samplerSource=SSM_Clamp_WorldGroupSettings
   samplerType=SAMPLERTYPE_LinearColor
   texture=/VanillaSoda/Gameplay/VanillaSodaWeapon/Art/Textures/T_UI_VanillaSoda_Shape
  }
  n22 = MaterialFunctionCall!(MSDF RGB=@n21.RGB) {materialFunction=/FortUILibrary/MaterialFunctions/Shapes/Modifiers/MF_UI_MSDFtoSDF}
  n23 = MaterialFunctionCall(DistanceField=@n22.SDF, Threshold=Constant() {r=0.45}) {materialFunction=&MF_UI_AntiAliasing}
  n24 = NamedRerouteDeclaration(None=MaterialFunctionCall(DistanceField=@n22.SDF, Threshold=Constant() {r=0.6}) {materialFunction=&MF_UI_AntiAliasing}.Outer) {name=Shape, nodeColor=<rgba 0 0 0 1>}
  n25 = ScalarParameter() {defaultValue=0.5, parameterName=ShapeShadowOpacity}
  NamedRerouteDeclaration(None=Subtract(
    A=NamedRerouteDeclaration(None=@n23.Outer) {name=Shape+Shadow, nodeColor=<rgba 0 0 0 1>}
    B=LinearInterpolate(A=@n23.Outer, B=@n24, Alpha=@n25)
   )) {name=ShapeShadowOnly, nodeColor=<rgba 0 0 0 1>}
 ]
}
