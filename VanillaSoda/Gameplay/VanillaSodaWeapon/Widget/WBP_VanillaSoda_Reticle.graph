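# Reticle widget for the VanillaSoda weapon, extending FortWeaponReticleExtensionWidgetBase.
# A centered Image in an Overlay draws the reticle through a UI material, whose parameters two widget animations drive.
# AnimAttack briefly pulses ReticleDistanceX out and back, while AnimCooldown ramps ProgressFill from empty to full with a constant CooldownColor.
# The event graph holds only unconnected stock events, so nothing in this Blueprint plays the animations.
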WidgetBlueprint /VanillaSoda/Gameplay/VanillaSodaWeapon/Widget/WBP_VanillaSoda_Reticle
animations [
 AnimCooldown {
  binding_count=1
  length_seconds=0.7
  object=&self:AnimCooldown
  sequence {bindings [Reticle {
     id=#1
     possessed_class=Image
     tracks [{
       type=MovieSceneWidgetMaterialTrack
       sections [{
         type=MovieSceneParameterSection
         end_frame=14
         start_frame=0
         channels [
          CooldownColor {type=MovieSceneScriptingFloatChannel, const [1,1]}
          ProgressFill {type=MovieSceneScriptingFloatChannel, keys [[0,0],[14,1]]}
         ]
        }]
      }]
    }]}
 }
 AnimAttack {
  binding_count=1
  length_seconds=0.4
  object=&self:AnimAttack
  sequence {bindings [Reticle {
     id=#2
     possessed_class=Image
     tracks [{
       type=MovieSceneWidgetMaterialTrack
       sections [{
         type=MovieSceneParameterSection
         end_frame=8
         start_frame=0
         channels [ReticleDistanceX {type=MovieSceneScriptingFloatChannel, keys [[0,0.35],[3.6923,0.42],[8,0.35]]}]
        }]
      }]
    }]}
 }
]
graphs [EventGraph {nodes [
   Event PreConstruct()
   Event Construct()
   Event Tick()
  ]}]
variables []
widgets {
 info {
  inheritedWidgetCount=1
  parentClass=/Script/FortniteUI.FortWeaponReticleExtensionWidgetBase
  rootWidgetClass=&Overlay
  widgetCount=3
 }
 widgets [
  Overlay_24 {widgetClassPath=&Overlay, widget=Overlay_24}
  Reticle {
   widgetClassPath=&Image
   bIsVariable=true
   parent=Overlay_24
   widget=Reticle
   properties {brush {
     bIsDynamicallyLoaded=false
     drawAs=Image
     imageType=FullColor
     mirroring=NoMirror
     resourceName=None
     resourceObject=/VanillaSoda/Gameplay/VanillaSodaWeapon/Art/Materials/M_UI_VanillaSoda_Reticle
     tiling=NoTile
     imageSize=<128 128>
     margin=<ltrb 0 0 0 0>
     outlineSettings {
      bUseBrushTransparency=false
      roundingType=HalfHeightRadius
      width=0
      color {colorUseRule=UseColor_Specified, specifiedColor=<rgba 0 0 0 0>}
      cornerRadii=<0 0 0 0>
     }
     tintColor {colorUseRule=UseColor_Specified, specifiedColor=<rgba 1 1 1 1>}
     uVRegion {bIsValid=false, max=<0 0>, min=<0 0>}
    }}
   slot {class=OverlaySlot, properties {horizontalAlignment=HAlign_Center, verticalAlignment=VAlign_Center}}
  }
  TutorialHighlight {widgetClassPath=&TutorialHighlight}
 ]
}
